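; Blueprint function Fortnite_ScreenShotCapture prepares the game for screenshot capture, apparently for high-quality marketing showcase images.
; It loops over a local list of console commands and executes each through the first player controller.
; Inside each loop pass it also sets the time of day to 13:00 and enables world rendering.
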(fn Fortnite_ScreenShotCapture
  (for _array_element (LocalVariables|GetConsoleCommands)
    (Development|ExecuteConsoleCommand 0 _array_element (Game|GetPlayerController 0))
    (TimeOfDay|SetTimeOfDay 0 13.0 :!class /Script/FortniteGame.FortKismetLibrary)
    (Rendering|SetEnableWorldRendering 0 true)))
